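; Blueprint for a Creative lava tile lava-fall piece that turns on its point light when play starts.
; On BeginPlay it gets the point light, sets it visible without propagating to children, and clears its hidden-in-game flag.
; The construction script is empty.
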(event EventBeginPlay
  (bind _pointlight (Variables|Default|GetPointLight))
  (Rendering|SetVisibility _pointlight true false)
  (Development|SetHiddeninGame _pointlight false false))

(fn ConstructionScript)
