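; Creative lava tile base parent Blueprint for placeable lava pieces in Creative mode.
; It overrides a few query functions: it can be targeted by linear movers and disallows invisible props.
; Physical-material footsteps are used only when the user option bIsLavaSurface is enabled.
; The construction script is empty.
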(fn IsTargetableByLinearMover
  (return true))

(fn ShouldInvisiblePropsBeAllowed
  (return false))

(fn ShouldUsePhysicalMaterialForFootstep
  (return (Variables|UserOptions|Basic|GetbIsLavaSurface)))

(fn ConstructionScript)
